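; Quest gameplay ability for the soccer slide emote quest that credits a player for dancing after a post-score slide.
; It activates from a gameplay event, noted as the player dancing, and checks the owning pawn for the SlideAfterScored tag.
; If the tag is present, it sends a custom stat event tagged DanceAfterSlide to the player's quest manager.
; The ability ends immediately in either case.
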(event Ability|EventActivateAbilityFromEvent
  (params (EventData :type Gameplay%20Event%20Data%20Structure%20%28by%20ref%29))
  ;[1 1 1] Activate ability - listen for when player Dances, then check if they have the SlideAfterScore tag
  (|CallParentFunction EventData)
  (if (GameplayTags|HasMatchingGameplayTag (Variables|GAVars|Data|GetOwningPlayerPawn) "(TagName=\"Athena.Quests.Soccer.SlideAfterScored\")")
    (Quest|SendComplexCustomStatEvent (Quests|GetQuestManager (Fortnite|GetFortPlayerControllerfromActor (Variables|GAVars|Data|GetOwningPlayerPawn)) "Invalid") 0 "(GameplayTags=((TagName=\"Athena.Quests.Soccer.DanceAfterSlide\")))" "(GameplayTags=)" 1)
    (Ability|EndAbility)
    (else
      (Ability|EndAbility))))
