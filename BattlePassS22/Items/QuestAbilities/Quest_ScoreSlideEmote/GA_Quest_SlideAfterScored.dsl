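; Gameplay ability for a soccer quest that credits the player for sliding after recently scoring a goal.
; On event activation, which the graph's note attributes to a slide, it checks the owning pawn for the RecentlyScored tag.
; If present, it applies the SlideAfterScored effect to the pawn and sends the matching custom stat event to the quest manager.
; The ability ends on either branch.
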(event Ability|EventActivateAbilityFromEvent
  (params (EventData :type Gameplay%20Event%20Data%20Structure%20%28by%20ref%29))
  (bind _owningplayerpawn (Variables|GAVars|Data|GetOwningPlayerPawn))
  ;[1 1 1] Activate ability - listen for when player slides, then check if they have the Scored Goal tag, then applies Slided After Scored tag
  (|CallParentFunction EventData)
  (if (GameplayTags|HasMatchingGameplayTag (Variables|GAVars|Data|GetOwningPlayerPawn) "(TagName=\"Athena.Quests.RecentlyScored\")")
    (GameplayEffects|ApplyGameplayEffectToSelf :self (Ability|GetFortAbilitySystemComponent _owningplayerpawn) :GameplayEffectClass (Variables|Default|GetSlideAfterScored) :Level 0.0)
    (Quest|SendComplexCustomStatEvent (Quests|GetQuestManager (Fortnite|GetFortPlayerControllerfromActor _owningplayerpawn) "Invalid") 0 "(GameplayTags=((TagName=\"Athena.Quests.Soccer.SlideAfterScored\")))" "(GameplayTags=)" 1)
    (Ability|EndAbility)
    (else
      (Ability|EndAbility))))
